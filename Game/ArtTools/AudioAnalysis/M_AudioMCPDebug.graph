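# Unlit debug material that displays audio-reactive parameter collection values as a 4x4 grid of numeric readouts.
# Floored, tiled UVs drive nested four-way lerps that pick one value per cell from amplitudes, music stems and spectrum channels.
# Each value hue-shifts a red base and is printed over it with DebugScalarValues.
# Both spectrum rows sample R in three cells and alpha in the fourth, leaving G and B unshown.
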Material /Game/ArtTools/AudioAnalysis/M_AudioMCPDebug
graph {
 properties {pixelDepthOffsetMode=PDOM_Legacy, shadingModel=MSM_Unlit, usedShadingModels=Unlit}
 nodes [
  n1 = Constant()
  Add(B=MaterialFunctionCall(1=Constant3Vector() {constant=<rgba 0 1 0 0>}, 2=@n1, 3=@n1, 4=Constant3Vector() {constant=<rgba 1 0 1 0>}, A=NamedRerouteUsage()) {materialFunction=&Lerp_Multiple_Float3}.Lerp 4 Inputs)
  Constant()
  n2 = CollectionParameter() {collection=&FortniteMaterialParameters, parameterName=EventAudio_Amplitude}
  n3 = CollectionParameter() {collection=&FortniteMaterialParameters, parameterName=EventAudio_Spectrum}
  n4 = MaterialFunctionCall(Color=@n3) {materialFunction=&SplitComponents}
  n5 = CollectionParameter() {collection=/Game/Athena/Events/AudioAnalysis/MPC_Audio_Reactive_Parameters, parameterName=Spectrum_Music}
  n6 = MaterialFunctionCall(Color=@n5) {materialFunction=&SplitComponents}
  n7 = MaterialFunctionCall!(
   1=MaterialFunctionCall(
    1=CollectionParameter() {collection=&MPC_Audio_Reactive_Parameters, parameterName=Amplitude_Master}
    2=CollectionParameter() {collection=&MPC_Audio_Reactive_Parameters, parameterName=Amplitude_Music}
    3=CollectionParameter() {collection=&MPC_Audio_Reactive_Parameters, parameterName=MusicStem_808}
    4=CollectionParameter() {collection=&MPC_Audio_Reactive_Parameters, parameterName=MusicStem_Kick}
    A=NamedRerouteUsage()
   ) {materialFunction=/Engine/Functions/Engine_MaterialFunctions02/Lerp_Multiple_Float}.Lerp 4 Inputs
   2=MaterialFunctionCall(
    1=CollectionParameter() {collection=&MPC_Audio_Reactive_Parameters, parameterName=MusicStem_Snare}
    2=CollectionParameter() {collection=&MPC_Audio_Reactive_Parameters, parameterName=MusicStem_Vocals}
    3=Constant()
    4=@n2
    A=NamedRerouteUsage()
   ) {materialFunction=&Lerp_Multiple_Float}.Lerp 4 Inputs
   3=MaterialFunctionCall(1=@n4.R, 2=@n4.R, 3=@n4.R, 4=ComponentMask(None=@n3) {a=true}, A=NamedRerouteUsage()) {materialFunction=&Lerp_Multiple_Float}.Lerp 4 Inputs
   4=MaterialFunctionCall(1=@n6.R, 2=@n6.R, 3=@n6.R, 4=ComponentMask(None=@n5) {a=true}, A=NamedRerouteUsage()) {materialFunction=&Lerp_Multiple_Float}.Lerp 4 Inputs
   A=NamedRerouteUsage()
  ) {materialFunction=&Lerp_Multiple_Float}
  n8 = TextureCoordinate() {uTiling=4, vTiling=4}
  n9 = Constant() {r=1}
  LinearInterpolate(
   A=MaterialFunctionCall(Hue Shift Percentage=@n7.Lerp 4 Inputs, Texture=Constant3Vector() {constant=<rgba 1 0 0 0>}) {materialFunction=&HueShift}.Result
   B=Constant3Vector()
   Alpha=MaterialFunctionCall(Number=@n7.Lerp 4 Inputs, MaximumNumberOfDigits=Constant() {r=3}, UVs=Fmod(A=@n8, B=@n9)) {materialFunction=&DebugScalarValues}.Result
  )
  n10 = Divide(A=Floor(None=TextureCoordinate() {uTiling=4, vTiling=4})) {constB=3}
  NamedRerouteDeclaration(None=ComponentMask(None=@n10) {r=true}) {name=CoordinatesX, nodeColor=<rgba 0.9453 0 1 1>}
  NamedRerouteDeclaration(None=ComponentMask(None=@n10) {g=true}) {name=CoordinatesY, nodeColor=<rgba 0.9453 0 1 1>}
  Reroute()
 ]
}
